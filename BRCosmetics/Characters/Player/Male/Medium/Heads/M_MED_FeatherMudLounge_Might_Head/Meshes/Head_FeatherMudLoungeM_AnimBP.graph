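# Animation Blueprint for the FeatherMudLounge Might male medium head cosmetic, applying a small fixed correction to the head skeleton's pose.
# The input pose passes through a component-space Transform Bone node that additively nudges teeth_lower along Z in bone space.
# A Control Rig node referencing the head's rig is present but not wired into the output.
# The event graph holds only unconnected default nodes.
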AnimBlueprint /BRCosmetics/Characters/Player/Male/Medium/Heads/M_MED_FeatherMudLounge_Might_Head/Meshes/Head_FeatherMudLoungeM_AnimBP
target_skeleton=&/Fortnite_Base_Head_FeatherMudLoungeM_Skeleton
graphs [
 AnimGraph {nodes [
   Control Rig() {
    alphaBoolBlend {blendOption=Linear}
    controlRigAssetReference {blueprintRigClass=&/M_MED_FeatherMudLounge_Might_Head_CtrlRig.M_MED_FeatherMudLounge_Might_Head_CtrlRig_C}
   }
   n1 = "Transform (Modify) Bone - Bone: teeth_lower"(ComponentPose=Local To Component(LocalPose=Input Pose()), Alpha=1, Translation="0,0,0.15", Rotation="0,0,0", Scale="1,1,1") {
    rotationMode=BMM_Additive
    rotationSpace=BCS_BoneSpace
    translationMode=BMM_Additive
    translationSpace=BCS_BoneSpace
    alphaBoolBlend {blendOption=Linear}
    boneToModify {boneName=teeth_lower}
    translation {z=0.15}
   }
   Output Pose(Result=Component To Local(ComponentPose=@n1))
  ]}
 EventGraph {nodes [
   TryGetPawnOwner()
   Event BlueprintUpdateAnimation()
  ]}
]
variables []
